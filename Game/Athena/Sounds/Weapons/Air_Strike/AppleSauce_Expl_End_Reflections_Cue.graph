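# Sound cue for the reflections layer at the end of an Air Strike (AppleSauce) explosion.
# It picks one of three reflection waves with equal weight through a Random node, then a Modulator varies pitch slightly downward.
# Overridden attenuation gives long-range falloff with distance-based low-pass filtering and reverb.
# Overridden concurrency caps instances at a few, stops the oldest, and scales down overlapping volume.
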SoundCue /Game/Athena/Sounds/Weapons/Air_Strike/AppleSauce_Expl_End_Reflections_Cue
root=Modulator_0
nodes [
 {id=Modulator_0, children [Random_1], properties {pitchMax=1, pitchMin=0.9}}
 {id=Random_1, children [WavePlayer_0,WavePlayer_1,WavePlayer_2], properties {weights [1,1,1]}}
 {id=WavePlayer_0, properties {soundWaveAssetPtr=&/AppleSauce_Expl_End_Reflections_01}}
 {id=WavePlayer_1, properties {soundWaveAssetPtr=&/AppleSauce_Expl_End_Reflections_02}}
 {id=WavePlayer_2, properties {soundWaveAssetPtr=&/AppleSauce_Expl_End_Reflections_03}}
]
properties {
 attenuationSettings=/Game/Sounds/Attenuation/Fort_AirStrike_Att
 bOverrideAttenuation=true
 bOverrideConcurrency=true
 maxDistance=32500
 soundClassObject=&Fort_WeaponRanged_Fire_3P_PSM
 volumeMultiplier=0.5
 attenuationOverrides {
  bAttenuateWithLPF=true
  falloffDistance=27500
  lPFFrequencyAtMax=800
  lPFRadiusMax=15000
  lPFRadiusMin=6000
  nonSpatializedRadiusStart=800
  reverbDistanceMax=16000
  reverbDistanceMin=1000
  stereoSpread=1000
  attenuationShapeExtents {x=5000}
 }
 concurrencyOverrides {maxCount=3, resolutionRule=StopOldest, volumeScale=0.75, volumeScaleAttackTime=0.1}
}
