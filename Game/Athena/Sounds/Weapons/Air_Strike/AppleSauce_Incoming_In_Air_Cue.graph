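# Sound cue for an incoming air strike (AppleSauce) in the air, routed to the third-person ranged weapon fire sound class.
# A mixer plays an incoming whoosh immediately and layers an air-burst wave after a fixed delay of just under a second.
# Overridden attenuation gives very long-range falloff with distance low-pass filtering, replacing the referenced air strike preset.
# Overridden concurrency caps simultaneous instances, stopping the oldest and scaling volume.
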SoundCue /Game/Athena/Sounds/Weapons/Air_Strike/AppleSauce_Incoming_In_Air_Cue
root=Mixer_1
nodes [
 {id=Delay_1, children [WavePlayer_5], properties {delayMax=0.9, delayMin=0.9}}
 {id=Mixer_1, children [WavePlayer_4,Delay_1], properties {inputVolume [1,1]}}
 {id=WavePlayer_4, properties {soundWaveAssetPtr=&/AppleSauce_Incoming_Whoosh_01}}
 {id=WavePlayer_5, properties {soundWaveAssetPtr=&/AppleSauce_Incoming_AirBurst_01}}
]
properties {
 attenuationSettings=/Game/Sounds/Attenuation/Fort_AirStrike_Att
 bOverrideAttenuation=true
 bOverrideConcurrency=true
 maxDistance=32500
 soundClassObject=&Fort_WeaponRanged_Fire_3P_PSM
 volumeMultiplier=0.7
 attenuationOverrides {
  bAttenuateWithLPF=true
  falloffDistance=27500
  lPFFrequencyAtMax=800
  lPFRadiusMax=15000
  lPFRadiusMin=6000
  nonSpatializedRadiusStart=800
  reverbDistanceMax=16000
  reverbDistanceMin=1000
  stereoSpread=1000
  attenuationShapeExtents {x=5000}
 }
 concurrencyOverrides {maxCount=3, resolutionRule=StopOldest, volumeScale=0.75, volumeScaleAttackTime=0.1}
}
