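# Sound cue for the incoming warning of the Air Strike weapon, codenamed AppleSauce.
# A single wave player feeds the warning wave straight to the output, with no randomization or mixing.
# The cue overrides attenuation inline for distance falloff, stereo spread and a non-spatialized near radius, and also sets low-pass radius and frequency values.
# It routes through the third-person ranged weapon fire sound class.
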SoundCue /Game/Athena/Sounds/Weapons/Air_Strike/AppleSauce_Incoming_Warning_Cue
root=WavePlayer_1
nodes [{id=WavePlayer_1, properties {soundWaveAssetPtr=&/AppleSauce_Warning_01}}]
properties {
 bOverrideAttenuation=true
 maxDistance=5000
 soundClassObject=&Fort_WeaponRanged_Fire_3P_PSM
 volumeMultiplier=1
 attenuationOverrides {
  falloffDistance=4000
  lPFFrequencyAtMax=100
  lPFFrequencyAtMin=1000
  lPFRadiusMax=1000
  lPFRadiusMin=500
  nonSpatializedRadiusStart=500
  stereoSpread=500
  attenuationShapeExtents {x=1000}
 }
}
