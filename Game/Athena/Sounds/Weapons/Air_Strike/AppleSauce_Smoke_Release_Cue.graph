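# Sound cue for the smoke release of the Air Strike weapon (AppleSauce), playing a single fuse smoke release wave.
# The wave player feeds a modulator left at its default settings, with the cue volume lowered.
# It routes to the explosives sound class and uses its own attenuation falloff and a concurrency override that prevents new instances once the limit is reached.
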SoundCue /Game/Athena/Sounds/Weapons/Air_Strike/AppleSauce_Smoke_Release_Cue
root=Modulator_1
nodes [
 {id=Modulator_1, children [WavePlayer_1]}
 {id=WavePlayer_1, properties {soundWaveAssetPtr=&/AppleSauce_Fuse_Smoke_Release_01}}
]
properties {
 bOverrideAttenuation=true
 bOverrideConcurrency=true
 maxDistance=2512
 soundClassObject=/Game/Sounds/FortSoundClassesAndModes/Fort_Explosives
 volumeMultiplier=0.7
 attenuationOverrides {falloffDistance=2000, attenuationShapeExtents {x=512}}
 concurrencyOverrides {resolutionRule=PreventNew}
}
