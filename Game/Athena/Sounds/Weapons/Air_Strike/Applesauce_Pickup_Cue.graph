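# Sound cue played when the Air Strike weapon (Applesauce) is picked up, classed as loot collection with weapon pickup attenuation.
# A quality-level root shares one mixer across the two first quality slots and another across the last two.
# Each mixer boosts its own wave player of an AppleSauce_Pickup_01 asset and leaves its second input empty.
# It references a shared pickup concurrency set and also carries single-instance, prevent-new override settings.
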SoundCue /Game/Athena/Sounds/Weapons/Air_Strike/Applesauce_Pickup_Cue
root=QualityLevel_0
nodes [
 {id=Mixer_2, children [WavePlayer_9,null], properties {inputVolume [2.5,0.5]}}
 {id=Mixer_3, children [WavePlayer_10, null], properties {inputVolume [2.5,0.5]}}
 {id=QualityLevel_0, children [Mixer_2,Mixer_2,Mixer_3,Mixer_3]}
 {id=WavePlayer_10, properties {soundWaveAssetPtr=&/AppleSauce_Pickup_01}}
 {id=WavePlayer_9, properties {soundWaveAssetPtr=&AppleSauce_Pickup_01}}
]
properties {
 attenuationSettings=&Fort_Weapon_Pickup_Att
 maxDistance=3500
 soundClassObject=&Fort_Loot_Collect
 volumeMultiplier=0.4615
 concurrencyOverrides {maxCount=1, resolutionRule=PreventNew}
 concurrencySet [&PickupMaximum_Oldest]
}
